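; CliffsStreaming is a spatial metrics data model preset that filters World Partition actor descriptor rows.
; ShouldKeepRow receives each actor's descriptor fields, such as class, bounds, runtime grid and data layers, plus its asset data.
; It returns only the spatially loaded flag, so always-loaded actors are dropped.
; Despite the preset's name, this filter applies no cliff-specific check on class, name or folder.
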(fn ShouldKeepRow
  (params (ActorDesc_Guid :type Guid%20Structure) (ActorDesc_NativeClass :type Object%20Class%20Reference) (ActorDesc_Class :type Soft%20Object%20Path%20Structure) (ActorDesc_Name :type Name) (ActorDesc_Label :type Name) (ActorDesc_Bounds :type Box%20Structure) (ActorDesc_RuntimeGrid :type Name) (ActorDesc_bIsSpatiallyLoaded :type Boolean) (ActorDesc_bActorIsEditorOnly :type Boolean) (ActorDesc_ActorPackage :type Name) (ActorDesc_ActorPath :type Name) (ActorDesc_FolderPath :type Name) (ActorDesc_DataLayerAssets :type Array%20of%20Soft%20Object%20Path%20Structures) (AssetData :type Asset%20Data%20Structure%20%28by%20ref%29) (Row :type Integer))
  (if true
    (return ActorDesc_bIsSpatiallyLoaded)))
